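# Sound cue for attaching to a zipline, playing a single Zipline_Attach_01 wave through one wave player.
# It routes to the gadgets and items sound class at reduced volume and uses a dedicated zipline attachment attenuation asset for spatialization.
# A concurrency override caps the number of simultaneous instances.
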SoundCue /Game/Sounds/Fort_GamePlay_Sounds/Zipline/SC_Zipline_Attach
root=WavePlayer_0
nodes [{id=WavePlayer_0, properties {soundWaveAssetPtr=&/Zipline_Attach_01}}]
properties {
 attenuationSettings=/Game/Sounds/Attenuation/Fort_Zipline_Attachment_Att
 bOverrideConcurrency=true
 maxDistance=10000
 soundClassObject=&Fort_Gadgets_Items
 volumeMultiplier=0.7
 concurrencyOverrides {maxCount=3}
}
