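# Looping zipline travel sound cue that blends mechanical, wind and brake-squeak layers for gadget audio, with attenuation and a concurrency limit.
# Continuous modulators use SpeedFactor to shift pitch and volume on two ascend mechanical loops and on descend mechanical and wind loops.
# IsBraking separately drives the squeak loop's volume and pitch.
# Nested mixers balance the layers, and a root Enveloper fades the whole mix in at start.
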SoundCue /Game/Sounds/Fort_GamePlay_Sounds/Zipline/SC_Zipline_Travel
root=Enveloper_0
nodes [
 {
  id=Enveloper_0
  children [Mixer_1]
  properties {pitchCurve {editorCurveData {keys [{time=0}]}}, volumeCurve {editorCurveData {keys [<at 0 0>,<at 0.2312 0.2052>,<at 0.6 1>]}}}
 }
 {id=Mixer_1, children [ModulatorContinuous_6, Mixer_2, Mixer_4], properties {inputVolume [0.65,1,1]}}
 {id=Mixer_2, children [ModulatorContinuous_3, ModulatorContinuous_5], properties {inputVolume [0.5,0.8]}}
 {id=Mixer_4, children [ModulatorContinuous_0, ModulatorContinuous_1], properties {inputVolume [0.85,1.8]}}
 {
  id=ModulatorContinuous_0
  children [WavePlayer_16]
  properties {
   pitchModulationParams {default=0, maxOutput=0.75, minInput=-1, minOutput=2, parameterName=SpeedFactor}
   volumeModulationParams {default=0, maxOutput=0, minInput=-1, minOutput=1, paramMode=MPM_Abs, parameterName=SpeedFactor}
  }
 }
 {
  id=ModulatorContinuous_1
  children [WavePlayer_17]
  properties {
   pitchModulationParams {default=0, maxOutput=2, minInput=-1, minOutput=0.8, parameterName=SpeedFactor}
   volumeModulationParams {default=0, maxOutput=0, minInput=-1, minOutput=1, parameterName=SpeedFactor}
  }
 }
 {
  id=ModulatorContinuous_3
  children [WavePlayer_2]
  properties {
   pitchModulationParams {default=0, maxOutput=1.25, minInput=-1, minOutput=0.65, parameterName=SpeedFactor}
   volumeModulationParams {default=0, minInput=0.25, parameterName=SpeedFactor}
  }
 }
 {
  id=ModulatorContinuous_5
  children [WavePlayer_3]
  properties {
   pitchModulationParams {default=0, maxOutput=1.15, minInput=-1, minOutput=0.75, parameterName=SpeedFactor}
   volumeModulationParams {default=0, minInput=0.25, parameterName=SpeedFactor}
  }
 }
 {
  id=ModulatorContinuous_6
  children [WavePlayer_4]
  properties {
   pitchModulationParams {default=0, maxOutput=1.1, minInput=0.25, minOutput=0.95, parameterName=IsBraking}
   volumeModulationParams {default=0, parameterName=IsBraking}
  }
 }
 {id=WavePlayer_16, properties {bLooping=true, soundWaveAssetPtr=&/Zipline_Descend_Mechanical_Loop_A}}
 {id=WavePlayer_17, properties {bLooping=true, soundWaveAssetPtr=&/Zipline_Descend_Wind_Loop}}
 {id=WavePlayer_2, properties {bLooping=true, soundWaveAssetPtr=&/Zipline_Ascend_Mechanical_Loop_A}}
 {id=WavePlayer_3, properties {bLooping=true, soundWaveAssetPtr=&/Zipline_Ascend_Mechanical_Loop_B}}
 {id=WavePlayer_4, properties {bLooping=true, soundWaveAssetPtr=&/Zipline_Brake_Loop_Squeak}}
]
properties {
 attenuationSettings=/Game/Sounds/Attenuation/Fort_Zipline_Travel_Att
 bOverrideConcurrency=true
 maxDistance=12700
 soundClassObject=&Fort_Gadgets_Items
 volumeMultiplier=0.6
 concurrencyOverrides {maxCount=3}
}
